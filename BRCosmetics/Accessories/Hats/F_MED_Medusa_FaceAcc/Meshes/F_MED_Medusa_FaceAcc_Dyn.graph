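# Physics asset for the Medusa face accessory cosmetic, defining collision and joint limits for its dynamic hair bones.
# Six hair chains (back, back-left, back-right, left, right, front-right) hang from the head as two or three linked bodies, mostly boxes with some capsule tips.
# Hair joints limit swing and tightly restrict twist, with wider swing on the right and back-right chains.
# Head, neck, upper spine and clavicles supply sphere and capsule bodies, joined by free-rotation constraints.
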PhysicsAsset /BRCosmetics/Accessories/Hats/F_MED_Medusa_FaceAcc/Meshes/F_MED_Medusa_FaceAcc_Dyn
bodies {
 clavicle_l [{shapeType=Capsule, length=7.2935, radius=7.3987, center=<5.5922 -0.9147 -1.9537>, rotation=<rot 83.3944 0 0>}]
 clavicle_r [{shapeType=Capsule, length=4.1754, radius=9.5935, center=<-4.9732 0.5965 3.1706>, rotation=<rot 83.3944 0 0>}]
 dyn_hair_bk_01 [{shapeType=Box, extentX=5.9004, extentY=1.8791, extentZ=7.5124, center=<3.2348 0 -0.0001>}]
 dyn_hair_bk_02 [{shapeType=Box, extentX=10.0425, extentY=1.8791, extentZ=7.5124, center=<5.0038 0 -0.0001>}]
 dyn_hair_bk_03 [{shapeType=Box, extentX=12.4028, extentY=2.5131, extentZ=7.5124, center=<6.4212 0 0.0001>}]
 dyn_hair_bk_l_01 [{shapeType=Box, extentX=5.9004, extentY=1.8791, extentZ=7.5124, center=<3.2348 0 -0.0001>}]
 dyn_hair_bk_l_02 [{shapeType=Box, extentX=5.9004, extentY=1.8791, extentZ=7.5124, center=<3.2348 0 -0.0001>}]
 dyn_hair_bk_l_03 [{shapeType=Capsule, length=3.5, radius=1.5, center=<0.484 0 -0.0003>}]
 dyn_hair_bk_r_01 [{shapeType=Box, extentX=5.9004, extentY=1.8791, extentZ=7.5124, center=<3.2348 0 -0.0001>}]
 dyn_hair_bk_r_02 [{shapeType=Box, extentX=5.9004, extentY=1.8791, extentZ=7.5124, center=<3.2348 0 -0.0001>}]
 dyn_hair_bk_r_03 [{shapeType=Capsule, length=4.8575, radius=1.5508, center=<1.7389 -0.0001 -0.0003>}]
 dyn_hair_fr_r_01 [{shapeType=Box, extentX=5.9004, extentY=1.8791, extentZ=7.5124, center=<3.2348 0 -0.0001>}]
 dyn_hair_fr_r_02 [{shapeType=Box, extentX=5.9004, extentY=1.8791, extentZ=7.5124, center=<3.2348 0 -0.0001>}]
 dyn_hair_l_01 [{shapeType=Box, extentX=5.9004, extentY=1.6203, extentZ=7.5124, center=<3.2348 0 -0.0001>}]
 dyn_hair_l_02 [{shapeType=Box, extentX=5.2096, extentY=1.8791, extentZ=7.5124, center=<3.2348 0 -0.0001>}]
 dyn_hair_l_03 [{shapeType=Capsule, length=5.375, radius=1.5, center=<0.9747 0 -0.0002>}]
 dyn_hair_r_01 [{shapeType=Box, extentX=5.9004, extentY=1.8791, extentZ=7.5124, center=<3.2348 0 -0.0001>}]
 dyn_hair_r_02 [{shapeType=Box, extentX=3.7066, extentY=1.8791, extentZ=10.1407, center=<2.3412 0 -0.0001>}]
 dyn_hair_r_03 [{shapeType=Capsule, length=8.1228, radius=1.5508, center=<0.9672 -0.0002 -0.0003>}]
 head [
  {shapeType=Sphere, radius=6.4177, center=<1.3282 -0.1209 -0.2823>}
  {shapeType=Capsule, length=5.9657, radius=5.522, center=<0.8118 1.1771 0>, rotation=<rot 90 0 0>}
 ]
 neck_01 [{shapeType=Capsule, length=8.2568, radius=5.3864, rotation=<rot 90 0 0>}]
 spine_05 [{shapeType=Capsule, length=0.0002, radius=9.7179, center=<6.8632 1.9165 0>}]
}
constraints [
 {bone1Name=dyn_hair_bk_l_01, bone2Name=head, swing1LimitDegrees=45, swing2LimitDegrees=45, twistLimitDegrees=5}
 {bone1Name=dyn_hair_bk_l_02, bone2Name=dyn_hair_bk_l_01, swing1LimitDegrees=45, swing2LimitDegrees=45, twistLimitDegrees=5}
 {bone1Name=dyn_hair_bk_l_03, bone2Name=dyn_hair_bk_l_02, swing1LimitDegrees=45, swing2LimitDegrees=45, twistLimitDegrees=5}
 {bone1Name=dyn_hair_bk_01, bone2Name=head, swing1LimitDegrees=60, swing2LimitDegrees=60, twistLimitDegrees=5}
 {bone1Name=dyn_hair_bk_02, bone2Name=dyn_hair_bk_01, swing1LimitDegrees=60, swing2LimitDegrees=60, twistLimitDegrees=5}
 {bone1Name=dyn_hair_bk_03, bone2Name=dyn_hair_bk_02, swing1LimitDegrees=60, swing2LimitDegrees=60, twistLimitDegrees=5}
 {bone1Name=dyn_hair_l_01, bone2Name=head, swing1LimitDegrees=45, swing2LimitDegrees=45, twistLimitDegrees=5}
 {bone1Name=dyn_hair_l_02, bone2Name=dyn_hair_l_01, swing1LimitDegrees=45, swing2LimitDegrees=45, twistLimitDegrees=5}
 {bone1Name=dyn_hair_l_03, bone2Name=dyn_hair_l_02, swing1LimitDegrees=45, swing2LimitDegrees=45, twistLimitDegrees=5}
 {bone1Name=dyn_hair_bk_r_01, bone2Name=head, swing1LimitDegrees=90, swing2LimitDegrees=60, twistLimitDegrees=10}
 {bone1Name=dyn_hair_bk_r_02, bone2Name=dyn_hair_bk_r_01, swing1LimitDegrees=90, swing2LimitDegrees=60, twistLimitDegrees=10}
 {bone1Name=dyn_hair_bk_r_03, bone2Name=dyn_hair_bk_r_02, swing1LimitDegrees=90, swing2LimitDegrees=60, twistLimitDegrees=10}
 {bone1Name=dyn_hair_r_01, bone2Name=head, swing1LimitDegrees=90, swing2LimitDegrees=60, twistLimitDegrees=5}
 {bone1Name=dyn_hair_r_02, bone2Name=dyn_hair_r_01, swing1LimitDegrees=90, swing2LimitDegrees=60, twistLimitDegrees=5}
 {bone1Name=dyn_hair_r_03, bone2Name=dyn_hair_r_02, swing1LimitDegrees=90, swing2LimitDegrees=60, twistLimitDegrees=5}
 {bone1Name=dyn_hair_fr_r_01, bone2Name=head, swing1LimitDegrees=45, swing2LimitDegrees=45, twistLimitDegrees=5}
 {bone1Name=dyn_hair_fr_r_02, bone2Name=dyn_hair_fr_r_01, swing1LimitDegrees=45, swing2LimitDegrees=45, twistLimitDegrees=5}
 {
  bone1Name=clavicle_l
  bone2Name=spine_05
  swing1LimitDegrees=45
  swing1Motion=Free
  swing2LimitDegrees=45
  swing2Motion=Free
  twistLimitDegrees=45
  twistMotion=Free
 }
 {
  bone1Name=clavicle_r
  bone2Name=spine_05
  swing1LimitDegrees=45
  swing1Motion=Free
  swing2LimitDegrees=45
  swing2Motion=Free
  twistLimitDegrees=45
  twistMotion=Free
 }
 {
  bone1Name=head
  bone2Name=neck_01
  swing1LimitDegrees=45
  swing1Motion=Free
  swing2LimitDegrees=45
  swing2Motion=Free
  twistLimitDegrees=45
  twistMotion=Free
 }
 {
  bone1Name=neck_01
  bone2Name=spine_05
  swing1LimitDegrees=45
  swing1Motion=Free
  swing2LimitDegrees=45
  swing2Motion=Free
  twistLimitDegrees=45
  twistMotion=Free
 }
]
