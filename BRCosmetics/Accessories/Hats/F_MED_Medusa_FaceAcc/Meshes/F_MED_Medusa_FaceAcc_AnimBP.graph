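# Animation Blueprint for the Medusa snake hat cosmetic, posing its snake bones over the pose copied from the attached parent mesh.
# A state machine switches between default, aim and in-motion (hunker) poses via intro and return clips, driven by bIsTargeting and bIsSnakeInMotion, with additive idles on top.
# The layered result feeds a head-space RigidBody simulation with skydiving external force.
# Two hat-bill AnimDynamics nodes are not connected to the output.
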AnimBlueprint /BRCosmetics/Accessories/Hats/F_MED_Medusa_FaceAcc/Meshes/F_MED_Medusa_FaceAcc_AnimBP
target_skeleton=&/F_MED_Medusa_FaceAcc_Skeleton
graphs [
 AnimGraph {nodes [
   n1 = Anim Dynamics - Bone: dyn_hat_bill_l(Alpha=1, ExternalForce="0,0,0") {
    angularDampingOverride=0.98
    bGravityOverrideInSimSpace=true
    bOverrideAngularDamping=true
    bUseGravityOverride=true
    alphaBoolBlend {blendOption=Linear}
    boundBone {boneName=dyn_hat_bill_l}
    gravityOverride {z=-980}
    physicsBodyDefinitions [{
      sphereCollisionRadius=0
      boundBone {boneName=dyn_hat_bill_l}
      boxExtents=<25 25 10>
      constraintSetup {twistAxis=AxisZ, angularLimitsMax=<0 10 0>, angularLimitsMin=<0 -10 0>, angularTarget=<0 0 0>}
      localJointOffset=<0 0 -3>
     }]
   }
   Anim Dynamics - Bone: dyn_hat_bill_r(ComponentPose=@n1, Alpha=1, ExternalForce="0,0,0") {
    angularDampingOverride=0.98
    bGravityOverrideInSimSpace=true
    bOverrideAngularDamping=true
    bUseGravityOverride=true
    alphaBoolBlend {blendOption=Linear}
    boundBone {boneName=dyn_hat_bill_r}
    gravityOverride {z=-980}
    physicsBodyDefinitions [{
      sphereCollisionRadius=0
      boundBone {boneName=dyn_hat_bill_r}
      boxExtents=<25 25 10>
      constraintSetup {twistAxis=AxisZ, angularLimitsMax=<0 10 0>, angularLimitsMin=<0 -10 0>, angularTarget=<0 0 0>}
      localJointOffset=<0 0 3>
     }]
   }
   n2 = Blend Poses by bool(
    BlendPose_0="Sequence Player (additive) 'Medusa_SnakeHat_Idle'"() {groupName=Idle, method=SyncGroup, sequence=/BRCosmetics/Animation/Game/MainPlayer/Hats/Medusa_SnakeHat_Idle}
    BlendPose_1="Sequence Player (additive) 'Medusa_SnakeHat_IdleHunker'"() {groupName=Idle, method=SyncGroup, sequence=/BRCosmetics/Animation/Game/MainPlayer/Hats/Medusa_SnakeHat_IdleHunker}
    bActiveValue=Get bIsTargeting()
   ) {blendType=Linear}
   n3 = "Sequence Player (additive) 'Medusa_SnakeHat_Idle'"() {groupName=Idle, method=SyncGroup, sequence=&Medusa_SnakeHat_Idle}
   n4 = Get bIsSnakeInMotion() -> (bIsSnakeInMotion:bool)
   AnimGraphNode_StateMachine_0 = SnakeTransition() {is=title AnimGraphNode_StateMachine}
   n5 = Layered blend per bone(
    BasePose=Copy Pose From Mesh() {bUseAttachedParent=true}
    BlendPoses_0=Apply Additive(Base=@AnimGraphNode_StateMachine_0, Additive=Blend Poses by bool(BlendPose_0=@n2, BlendPose_1=@n3, bActiveValue=@n4) {blendType=Linear}) {alphaBoolBlend {blendInTime=0.2, blendOption=Linear, blendOutTime=0.2}}
   ) {layerSetup [{branchFilters [
       {boneName=dyn_snake_head_l_01}
       {boneName=dyn_snake_head_r_01}
       {boneName=dyn_snake_head_bk_01}
       {boneName=dyn_snake_tail_r_01}
       {boneName=dyn_snake_tail_l_01}
       {boneName=dyn_snake_body_01}
       {boneName=dyn_snake_body_02}
       {boneName=dyn_snake_body_03}
       {boneName=dyn_snake_body_04}
       {boneName=dyn_snake_body_05}
       {boneName=dyn_snake_body_06}
       {boneName=dyn_snake_body_07}
       {boneName=dyn_snake_body_08}
       {boneName=dyn_snake_body_09}
       {boneName=dyn_snake_body_10}
       {boneName=dyn_snake_body_11}
      ]}]}
   n6 = RigidBody(
    ComponentPose=Local To Component(LocalPose=@n5)
    ExternalForce=Break Fort Anim Input Skydiving External Force(FortAnimInput_SkydivingExternalForce=Get SkydivingExternalForceInput()) -> (
     bUseSkydivingVectorForce=~
     SkydivingVectorForce=~
     HeadToPelvisDirection=~
     FloatingMultiplier=~
     FloatingAdditive=~
     DivingMultiplier=~
     DivingAdditive=~
     ParachutingMultiplier=~
     ParachutingAdditive=~
    ).SkydivingVectorForce
   ) {
    bOverrideWorldGravity=true
    overridePhysicsAsset=&/F_MED_Medusa_FaceAcc_Dyn
    simulationSpace=BaseBoneSpace
    alphaBoolBlend {blendOption=Linear}
    baseBoneRef {boneName=head}
    componentAppliedLinearAccClamp=<2000 2000 2000>
    componentLinearAccScale=<0.2 0.5 0.8>
    componentLinearVelScale=<0.2 0.5 0.8>
    overrideWorldGravity {z=-980}
   }
   Output Pose(Result=Component To Local(ComponentPose=@n6)) {layerGroup=None}
  ]}
 SnakeTransition {
  entry=Default
  path=AnimGraph.AnimGraphNode_StateMachine_0.SnakeTransition
  ids=id
  nodes [
   AnimStateNode_0 = Default() {is=title AnimStateNode}
   AnimStateNode_1 = ADS_Intro() {is=title AnimStateNode}
   AnimStateNode_2 = ADS() {is=title AnimStateNode}
   AnimStateNode_3 = Hunker_Intro() {is=title AnimStateNode}
   AnimStateNode_4 = Hunker() {is=title AnimStateNode}
   AnimStateNode_5 = HunterToIdle() {is=title AnimStateNode}
   AnimStateTransitionNode_0 = AnimStateTransitionNode()
   AnimStateTransitionNode_1 = AnimStateTransitionNode() {bAutomaticRuleBasedOnSequencePlayerInState=true}
   AnimStateTransitionNode_10 = AnimStateTransitionNode()
   AnimStateTransitionNode_11 = AnimStateTransitionNode() {bAutomaticRuleBasedOnSequencePlayerInState=true, crossfadeDuration=0}
   AnimStateTransitionNode_12 = AnimStateTransitionNode()
   AnimStateTransitionNode_13 = AnimStateTransitionNode()
   AnimStateTransitionNode_2 = AnimStateTransitionNode()
   AnimStateTransitionNode_3 = AnimStateTransitionNode()
   AnimStateTransitionNode_4 = AnimStateTransitionNode() {bAutomaticRuleBasedOnSequencePlayerInState=true}
   AnimStateTransitionNode_5 = AnimStateTransitionNode()
   AnimStateTransitionNode_6 = AnimStateTransitionNode()
   AnimStateTransitionNode_7 = AnimStateTransitionNode()
   AnimStateTransitionNode_8 = AnimStateTransitionNode()
  ]
  links [
   {
    dst=Default
    rule=AnimGraph.AnimGraphNode_StateMachine_0.SnakeTransition.AnimStateTransitionNode_2.Transition
    src=ADS
    via=AnimStateTransitionNode_2
   }
   {
    dst=Hunker_Intro
    rule=AnimGraph.AnimGraphNode_StateMachine_0.SnakeTransition.AnimStateTransitionNode_8.Transition
    src=ADS
    via=AnimStateTransitionNode_8
   }
   {
    dst=ADS
    rule=AnimGraph.AnimGraphNode_StateMachine_0.SnakeTransition.AnimStateTransitionNode_1.Transition
    src=ADS_Intro
    via=AnimStateTransitionNode_1
   }
   {
    dst=Hunker_Intro
    rule=AnimGraph.AnimGraphNode_StateMachine_0.SnakeTransition.AnimStateTransitionNode_7.Transition
    src=ADS_Intro
    via=AnimStateTransitionNode_7
   }
   {
    dst=ADS_Intro
    rule=AnimGraph.AnimGraphNode_StateMachine_0.SnakeTransition.AnimStateTransitionNode_0.Transition
    src=Default
    via=AnimStateTransitionNode_0
   }
   {
    dst=Hunker_Intro
    rule=AnimGraph.AnimGraphNode_StateMachine_0.SnakeTransition.AnimStateTransitionNode_3.Transition
    src=Default
    via=AnimStateTransitionNode_3
   }
   {
    dst=ADS_Intro
    rule=AnimGraph.AnimGraphNode_StateMachine_0.SnakeTransition.AnimStateTransitionNode_6.Transition
    src=Hunker
    via=AnimStateTransitionNode_6
   }
   {
    dst=HunterToIdle
    rule=AnimGraph.AnimGraphNode_StateMachine_0.SnakeTransition.AnimStateTransitionNode_10.Transition
    src=Hunker
    via=AnimStateTransitionNode_10
   }
   {
    dst=ADS_Intro
    rule=AnimGraph.AnimGraphNode_StateMachine_0.SnakeTransition.AnimStateTransitionNode_5.Transition
    src=Hunker_Intro
    via=AnimStateTransitionNode_5
   }
   {
    dst=Hunker
    rule=AnimGraph.AnimGraphNode_StateMachine_0.SnakeTransition.AnimStateTransitionNode_4.Transition
    src=Hunker_Intro
    via=AnimStateTransitionNode_4
   }
   {
    dst=ADS_Intro
    rule=AnimGraph.AnimGraphNode_StateMachine_0.SnakeTransition.AnimStateTransitionNode_12.Transition
    src=HunterToIdle
    via=AnimStateTransitionNode_12
   }
   {
    dst=Default
    rule=AnimGraph.AnimGraphNode_StateMachine_0.SnakeTransition.AnimStateTransitionNode_11.Transition
    src=HunterToIdle
    via=AnimStateTransitionNode_11
   }
   {
    dst=Hunker_Intro
    rule=AnimGraph.AnimGraphNode_StateMachine_0.SnakeTransition.AnimStateTransitionNode_13.Transition
    src=HunterToIdle
    via=AnimStateTransitionNode_13
   }
  ]
 }
 Default {
  path=AnimGraph.AnimGraphNode_StateMachine_0.SnakeTransition.AnimStateNode_0.Default
  nodes [Output Animation Pose(Result=Sequence Player 'Medusa_SnakeHat_Idle_Pose'() {sequence=/BRCosmetics/Animation/Game/MainPlayer/Hats/Medusa_SnakeHat_Idle_Pose})]
 }
 ADS_Intro {
  path=AnimGraph.AnimGraphNode_StateMachine_0.SnakeTransition.AnimStateNode_1.ADS_Intro
  nodes [Output Animation Pose(Result=Sequence Player 'Medusa_SnakeHat_ADS'() {bLoopAnimation=false, sequence=/BRCosmetics/Animation/Game/MainPlayer/Hats/Medusa_SnakeHat_ADS})]
 }
 ADS {
  path=AnimGraph.AnimGraphNode_StateMachine_0.SnakeTransition.AnimStateNode_2.ADS
  nodes [Output Animation Pose(Result=Sequence Player 'Medusa_SnakeHat_ADS_pose'() {sequence=/BRCosmetics/Animation/Game/MainPlayer/Hats/Medusa_SnakeHat_ADS_pose})]
 }
 Hunker_Intro {
  path=AnimGraph.AnimGraphNode_StateMachine_0.SnakeTransition.AnimStateNode_3.Hunker_Intro
  nodes [Output Animation Pose(Result=Sequence Player 'Medusa_SnakeHat_InMotion'() {bLoopAnimation=false, sequence=/BRCosmetics/Animation/Game/MainPlayer/Hats/Medusa_SnakeHat_InMotion})]
 }
 Hunker {
  path=AnimGraph.AnimGraphNode_StateMachine_0.SnakeTransition.AnimStateNode_4.Hunker
  nodes [Output Animation Pose(Result=Sequence Player 'Medusa_SnakeHat_InMotion_Pose'() {sequence=/BRCosmetics/Animation/Game/MainPlayer/Hats/Medusa_SnakeHat_InMotion_Pose})]
 }
 HunterToIdle {
  path=AnimGraph.AnimGraphNode_StateMachine_0.SnakeTransition.AnimStateNode_5.HunterToIdle
  nodes [Output Animation Pose(Result=Sequence Player 'Medusa_SnakeHat_InMotionToIdle'() {bLoopAnimation=false, sequence=/BRCosmetics/Animation/Game/MainPlayer/Hats/Medusa_SnakeHat_InMotionToIdle})]
 }
 Transition {
  path=AnimGraph.AnimGraphNode_StateMachine_0.SnakeTransition.AnimStateTransitionNode_0.Transition
  nodes [Result(bCanEnterTransition=Get bIsTargeting())]
 }
 Transition {path=AnimGraph.AnimGraphNode_StateMachine_0.SnakeTransition.AnimStateTransitionNode_1.Transition, nodes [Result()]}
 Transition {
  path=AnimGraph.AnimGraphNode_StateMachine_0.SnakeTransition.AnimStateTransitionNode_10.Transition
  nodes [Result(bCanEnterTransition=NOT Boolean(A=Get bIsSnakeInMotion()))]
 }
 Transition {path=AnimGraph.AnimGraphNode_StateMachine_0.SnakeTransition.AnimStateTransitionNode_11.Transition, nodes [Result()]}
 Transition {
  path=AnimGraph.AnimGraphNode_StateMachine_0.SnakeTransition.AnimStateTransitionNode_12.Transition
  nodes [Result(bCanEnterTransition=Get bIsTargeting())]
 }
 Transition {
  path=AnimGraph.AnimGraphNode_StateMachine_0.SnakeTransition.AnimStateTransitionNode_13.Transition
  nodes [
   n1 = Get bIsTargeting()
   Result(bCanEnterTransition=AND Boolean(A=Get bIsSnakeInMotion(), B=NOT Boolean(A=@n1)))
  ]
 }
 Transition {
  path=AnimGraph.AnimGraphNode_StateMachine_0.SnakeTransition.AnimStateTransitionNode_2.Transition
  nodes [Result(bCanEnterTransition=NOT Boolean(A=Get bIsTargeting()))]
 }
 Transition {
  path=AnimGraph.AnimGraphNode_StateMachine_0.SnakeTransition.AnimStateTransitionNode_3.Transition
  nodes [Result(bCanEnterTransition=Get bIsSnakeInMotion())]
 }
 Transition {path=AnimGraph.AnimGraphNode_StateMachine_0.SnakeTransition.AnimStateTransitionNode_4.Transition, nodes [Result()]}
 Transition {
  path=AnimGraph.AnimGraphNode_StateMachine_0.SnakeTransition.AnimStateTransitionNode_5.Transition
  nodes [Result(bCanEnterTransition=Get bIsTargeting())]
 }
 Transition {
  path=AnimGraph.AnimGraphNode_StateMachine_0.SnakeTransition.AnimStateTransitionNode_6.Transition
  nodes [Result(bCanEnterTransition=Get bIsTargeting())]
 }
 Transition {
  path=AnimGraph.AnimGraphNode_StateMachine_0.SnakeTransition.AnimStateTransitionNode_7.Transition
  nodes [
   n1 = Get bIsTargeting()
   Result(bCanEnterTransition=AND Boolean(A=Get bIsSnakeInMotion(), B=NOT Boolean(A=@n1)))
  ]
 }
 Transition {
  path=AnimGraph.AnimGraphNode_StateMachine_0.SnakeTransition.AnimStateTransitionNode_8.Transition
  nodes [
   n1 = Get bIsTargeting()
   Result(bCanEnterTransition=AND Boolean(A=Get bIsSnakeInMotion(), B=NOT Boolean(A=@n1)))
  ]
 }
 EventGraph {links [], nodes []}
]
variables []
